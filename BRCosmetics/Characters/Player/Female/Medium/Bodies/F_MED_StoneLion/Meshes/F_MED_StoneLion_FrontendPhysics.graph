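# Physics asset for the StoneLion female medium body's frontend variant, providing capsule bodies and joint limits for its dangling belt and cloth bones.
# Pelvis, lower spine, legs and left upper arm carry body capsules, with the arm left unconstrained.
# Left, right and back cloth chains lock their roots to the pelvis, with A/B chains also locking their second link and lower links swinging within limits.
# Two belt branches hang from a tightly limited belt base.
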PhysicsAsset /BRCosmetics/Characters/Player/Female/Medium/Bodies/F_MED_StoneLion/Meshes/F_MED_StoneLion_FrontendPhysics
bodies {
 calf_l [{shapeType=Capsule, length=44.3406, radius=9.8239, center=<-27.2654 0.5586 0.467>, rotation=<rot 0 79.8132 -88.7734>}]
 calf_r [{shapeType=Capsule, length=44.5078, radius=9.8215, center=<27.3794 -0.1258 -0.3504>, rotation=<rot 0 78.8394 -88.3108>}]
 dyn_Belt_a01 [{
   shapeType=Capsule
   length=4.2964
   radius=3.1285
   shapeName=dyn_Belt_a01_capsule
   center=<4.808 -0.5895 0.4991>
   rotation=<rot 0 -83.209 -90.1761>
  }]
 dyn_Belt_a02 [{
   shapeType=Capsule
   length=3.9904
   radius=2.1348
   shapeName=dyn_Belt_a02_capsule
   center=<4.5854 -0.4922 0.7293>
   rotation=<rot 0 87.9916 -81.618>
  }]
 dyn_Belt_b01 [{shapeType=Capsule, length=66.2982, radius=2.9127, center=<14.9237 -11.0092 -3.6603>, rotation=<rot 35.2511 -121.9118 -58.3128>}]
 dyn_Belt_b01_base [{shapeType=Capsule, length=0, radius=6.747, center=<7.2284 3.163 5.6711>, rotation=<rot 0 33.7781 1.1755>}]
 dyn_Belt_base [{shapeType=Capsule, length=0, radius=6.747, center=<7.2284 3.163 5.6711>, rotation=<rot 0 33.7781 1.1755>}]
 dyn_Cloth_volume_bk [{
   shapeType=Capsule
   length=7
   radius=4.5
   shapeName=dyn_Cloth_volume_bk_capsule
   center=<-4.7395 1.8107 0.2328>
   rotation=<rot -87.1875 179.9999 -180>
  }]
 dyn_Cloth_volume_bk01 [{
   shapeType=Capsule
   length=7
   radius=4.5
   shapeName=dyn_Cloth_volume_bk01_capsule
   center=<-4.7395 -1.6471 0.2328>
   rotation=<rot 2.8125 179.9997 180>
  }]
 dyn_Cloth_volume_l_A [{
   shapeType=Capsule
   length=7
   radius=4.5
   shapeName=dyn_Cloth_volume_l_A_capsule
   center=<-4.7395 1.8107 0.2328>
   rotation=<rot -87.1875 -179.9998 179.9999>
  }]
 dyn_Cloth_volume_l_A01 [{
   shapeType=Capsule
   length=7
   radius=4.5
   shapeName=dyn_Cloth_volume_l_A01_capsule
   center=<-4.7395 1.8107 0.2328>
   rotation=<rot -87.1875 -179.9998 179.9999>
  }]
 dyn_Cloth_volume_l_A02 [{
   shapeType=Capsule
   length=7
   radius=4.5
   shapeName=dyn_Cloth_volume_l_A02_capsule
   center=<-4.7395 -2.6327 0.2328>
   rotation=<rot -87.1875 -179.9998 179.9999>
  }]
 dyn_Cloth_volume_l_A03 [{
   shapeType=Capsule
   length=7
   radius=4.5
   shapeName=dyn_Cloth_volume_l_A03_capsule
   center=<-4.7395 -2.6327 0.2328>
   rotation=<rot -87.1875 -179.9998 179.9999>
  }]
 dyn_Cloth_volume_l_B [{
   shapeType=Capsule
   length=7
   radius=4.5
   shapeName=dyn_Cloth_volume_l_B_capsule
   center=<-4.7395 1.8107 0.2328>
   rotation=<rot -87.1876 -179.9998 179.9998>
  }]
 dyn_Cloth_volume_l_B01 [{
   shapeType=Capsule
   length=7
   radius=4.5
   shapeName=dyn_Cloth_volume_l_B01_capsule
   center=<-4.7395 3.2451 0.2328>
   rotation=<rot -87.1876 -179.9998 179.9998>
  }]
 dyn_Cloth_volume_l_B02 [{
   shapeType=Capsule
   length=7
   radius=4.5
   shapeName=dyn_Cloth_volume_l_B02_capsule
   center=<-4.7395 -2.6327 0.2328>
   rotation=<rot -87.1876 -179.9998 179.9998>
  }]
 dyn_Cloth_volume_l_B03 [{
   shapeType=Capsule
   length=7
   radius=4.5
   shapeName=dyn_Cloth_volume_l_B03_capsule
   center=<-4.7395 -2.6327 0.2328>
   rotation=<rot -87.1876 -179.9998 179.9998>
  }]
 dyn_Cloth_volume_l_C [{
   shapeType=Capsule
   length=7
   radius=4.5
   shapeName=dyn_Cloth_volume_l_C_capsule
   center=<-4.7395 1.8107 0.2328>
   rotation=<rot -87.1875 -180 179.9999>
  }]
 dyn_Cloth_volume_l_C01 [{
   shapeType=Capsule
   length=7
   radius=4.5
   shapeName=dyn_Cloth_volume_l_C01_capsule
   center=<-4.7395 -1.6472 0.2328>
   rotation=<rot 2.8124 179.9998 180>
  }]
 dyn_Cloth_volume_l_C02 [{
   shapeType=Capsule
   length=12.5
   radius=5.5
   shapeName=dyn_Cloth_volume_l_C02_capsule
   center=<-4.6786 -2.3121 1.4729>
   rotation=<rot -2.8125 -0.0002 0>
  }]
 dyn_Cloth_volume_l_C03 [{
   shapeType=Capsule
   length=14
   radius=5.5
   shapeName=dyn_Cloth_volume_l_C03_capsule
   center=<-4.6786 -2.254 1.4729>
   rotation=<rot -2.8125 -0.0002 0>
  }]
 dyn_Cloth_volume_r_A [{
   shapeType=Capsule
   length=7
   radius=4.5
   shapeName=dyn_Cloth_volume_r_A_capsule
   center=<-4.7395 1.8107 0.2328>
   rotation=<rot -87.1875 180 -179.9999>
  }]
 dyn_Cloth_volume_r_A01 [{
   shapeType=Capsule
   length=7
   radius=4.5
   shapeName=dyn_Cloth_volume_r_A01_capsule
   center=<-4.7395 1.8107 0.2328>
   rotation=<rot -87.1875 180 -179.9999>
  }]
 dyn_Cloth_volume_r_A02 [{
   shapeType=Capsule
   length=7
   radius=4.5
   shapeName=dyn_Cloth_volume_r_A02_capsule
   center=<-4.7395 -2.6327 0.2328>
   rotation=<rot -87.1875 180 -179.9999>
  }]
 dyn_Cloth_volume_r_A03 [{
   shapeType=Capsule
   length=7
   radius=4.5
   shapeName=dyn_Cloth_volume_r_A03_capsule
   center=<-4.7395 -2.6327 0.2328>
   rotation=<rot -87.1875 180 -179.9999>
  }]
 dyn_Cloth_volume_r_B [{
   shapeType=Capsule
   length=7
   radius=4.5
   shapeName=dyn_Cloth_volume_r_B_capsule
   center=<-4.7395 1.8107 0.2328>
   rotation=<rot -87.1874 179.9998 -179.9998>
  }]
 dyn_Cloth_volume_r_B01 [{
   shapeType=Capsule
   length=7
   radius=4.5
   shapeName=dyn_Cloth_volume_r_B01_capsule
   center=<-4.7395 1.8107 0.2328>
   rotation=<rot -87.1874 179.9998 -179.9998>
  }]
 dyn_Cloth_volume_r_B02 [{
   shapeType=Capsule
   length=7
   radius=4.5
   shapeName=dyn_Cloth_volume_r_B02_capsule
   center=<-4.7395 -2.6327 0.2328>
   rotation=<rot -87.1874 179.9998 -179.9998>
  }]
 dyn_Cloth_volume_r_B03 [{
   shapeType=Capsule
   length=7
   radius=4.5
   shapeName=dyn_Cloth_volume_r_B03_capsule
   center=<-4.7395 -2.6327 0.2328>
   rotation=<rot -87.1874 179.9998 -179.9998>
  }]
 dyn_Cloth_volume_r_C [{
   shapeType=Capsule
   length=7
   radius=4.5
   shapeName=dyn_Cloth_volume_r_C_capsule
   center=<-4.7395 1.8107 0.2328>
   rotation=<rot -87.1875 179.9998 -179.9999>
  }]
 dyn_Cloth_volume_r_C01 [{
   shapeType=Capsule
   length=7
   radius=4.5
   shapeName=dyn_Cloth_volume_r_C01_capsule
   center=<-4.7395 -1.6472 0.2328>
   rotation=<rot 2.8126 179.9998 -180>
  }]
 dyn_Cloth_volume_r_C02 [{
   shapeType=Capsule
   length=18
   radius=5.5
   shapeName=dyn_Cloth_volume_r_C02_capsule
   center=<-4.7395 -3.8058 0.2328>
   rotation=<rot -2.8126 -0.0002 0>
  }]
 dyn_Cloth_volume_r_C03 [{
   shapeType=Capsule
   length=15
   radius=6.5
   shapeName=dyn_Cloth_volume_r_C03_capsule
   center=<-3.0959 -3.9719 0.1325>
   rotation=<rot -2.8126 -0.0002 0>
  }]
 pelvis [{shapeType=Capsule, length=4.0447, radius=14.2682, center=<1.9277 -0.3348 0.4765>, rotation=<rot -3.4833 -90.0002 177.7238>}]
 spine_01 [{shapeType=Capsule, length=7.6347, radius=12.9783, center=<13.0152 0.5993 -0.7628>, rotation=<rot 0 179.0484 9.8028>}]
 thigh_l [{shapeType=Capsule, length=28.8814, radius=7.9804, center=<-23.0185 -1.4423 1.2568>, rotation=<rot 0 -89.5782 -90.721>}]
 thigh_r [{shapeType=Capsule, length=28.4675, radius=8.3276, center=<23.0383 0.6925 -1.0598>, rotation=<rot 0 -88.643 -92.1258>}]
 upperarm_l [{shapeType=Capsule, length=18.5, radius=8.5, shapeName=upperarm_l_capsule, center=<12.2394 0 -1.705>, rotation=<rot -87.1875 0.001 -0.001>}]
}
constraints [
 {bone1Name=spine_01, bone2Name=pelvis, swing1LimitDegrees=45, swing2LimitDegrees=45, twistLimitDegrees=45}
 {bone1Name=thigh_l, bone2Name=pelvis, swing1LimitDegrees=45, swing2LimitDegrees=45, twistLimitDegrees=45}
 {bone1Name=calf_l, bone2Name=thigh_l, swing1LimitDegrees=45, swing2LimitDegrees=45, twistLimitDegrees=45}
 {bone1Name=thigh_r, bone2Name=pelvis, swing1LimitDegrees=45, swing2LimitDegrees=45, twistLimitDegrees=45}
 {bone1Name=calf_r, bone2Name=thigh_r, swing1LimitDegrees=45, swing2LimitDegrees=45, twistLimitDegrees=45}
 {bone1Name=dyn_Belt_base, bone2Name=pelvis, swing1LimitDegrees=5, swing2LimitDegrees=10, twistLimitDegrees=5}
 {bone1Name=dyn_Belt_b01, bone2Name=dyn_Belt_b01_base, swing1LimitDegrees=15, swing2LimitDegrees=10, twistLimitDegrees=5}
 {
  bone1Name=dyn_Belt_b01_base
  bone2Name=dyn_Belt_base
  swing1LimitDegrees=45
  swing1Motion=Locked
  swing2LimitDegrees=45
  swing2Motion=Locked
  twistLimitDegrees=45
  twistMotion=Locked
 }
 {bone1Name=dyn_Cloth_volume_l_A03, bone2Name=dyn_Cloth_volume_l_A02, swing1LimitDegrees=30, swing2LimitDegrees=15, twistLimitDegrees=15}
 {bone1Name=dyn_Cloth_volume_l_A02, bone2Name=dyn_Cloth_volume_l_A01, swing1LimitDegrees=30, swing2LimitDegrees=15, twistLimitDegrees=15}
 {bone1Name=dyn_Cloth_volume_l_B03, bone2Name=dyn_Cloth_volume_l_B02, swing1LimitDegrees=30, swing2LimitDegrees=15, twistLimitDegrees=15}
 {bone1Name=dyn_Cloth_volume_l_B02, bone2Name=dyn_Cloth_volume_l_B01, swing1LimitDegrees=30, swing2LimitDegrees=15, twistLimitDegrees=15}
 {
  bone1Name=dyn_Cloth_volume_l_B01
  bone2Name=dyn_Cloth_volume_l_B
  swing1LimitDegrees=45
  swing1Motion=Locked
  swing2LimitDegrees=45
  swing2Motion=Locked
  twistLimitDegrees=45
  twistMotion=Locked
 }
 {bone1Name=dyn_Cloth_volume_l_C03, bone2Name=dyn_Cloth_volume_l_C02, swing1LimitDegrees=45, swing2LimitDegrees=45, twistLimitDegrees=45}
 {bone1Name=dyn_Cloth_volume_l_C02, bone2Name=dyn_Cloth_volume_l_C01, swing1LimitDegrees=30, swing2LimitDegrees=15, twistLimitDegrees=15}
 {bone1Name=dyn_Cloth_volume_l_C01, bone2Name=dyn_Cloth_volume_l_C, swing1LimitDegrees=15, swing2LimitDegrees=5, twistLimitDegrees=5}
 {bone1Name=dyn_Cloth_volume_r_A02, bone2Name=dyn_Cloth_volume_r_A01, swing1LimitDegrees=30, swing2LimitDegrees=15, twistLimitDegrees=15}
 {bone1Name=dyn_Cloth_volume_r_A03, bone2Name=dyn_Cloth_volume_r_A02, swing1LimitDegrees=30, swing2LimitDegrees=15, twistLimitDegrees=15}
 {
  bone1Name=dyn_Cloth_volume_r_A01
  bone2Name=dyn_Cloth_volume_r_A
  swing1LimitDegrees=45
  swing1Motion=Locked
  swing2LimitDegrees=45
  swing2Motion=Locked
  twistLimitDegrees=45
  twistMotion=Locked
 }
 {
  bone1Name=dyn_Cloth_volume_l_A01
  bone2Name=dyn_Cloth_volume_l_A
  swing1LimitDegrees=45
  swing1Motion=Locked
  swing2LimitDegrees=45
  swing2Motion=Locked
  twistLimitDegrees=45
  twistMotion=Locked
 }
 {bone1Name=dyn_Cloth_volume_bk01, bone2Name=dyn_Cloth_volume_bk, swing1LimitDegrees=15, swing2LimitDegrees=5, twistLimitDegrees=5}
 {
  bone1Name=dyn_Cloth_volume_bk
  bone2Name=pelvis
  swing1LimitDegrees=45
  swing1Motion=Locked
  swing2LimitDegrees=45
  swing2Motion=Locked
  twistLimitDegrees=45
  twistMotion=Locked
 }
 {
  bone1Name=dyn_Cloth_volume_l_A
  bone2Name=pelvis
  swing1LimitDegrees=45
  swing1Motion=Locked
  swing2LimitDegrees=45
  swing2Motion=Locked
  twistLimitDegrees=45
  twistMotion=Locked
 }
 {
  bone1Name=dyn_Cloth_volume_l_B
  bone2Name=pelvis
  swing1LimitDegrees=45
  swing1Motion=Locked
  swing2LimitDegrees=45
  swing2Motion=Locked
  twistLimitDegrees=45
  twistMotion=Locked
 }
 {
  bone1Name=dyn_Cloth_volume_l_C
  bone2Name=pelvis
  swing1LimitDegrees=45
  swing1Motion=Locked
  swing2LimitDegrees=45
  swing2Motion=Locked
  twistLimitDegrees=45
  twistMotion=Locked
 }
 {
  bone1Name=dyn_Cloth_volume_r_A
  bone2Name=pelvis
  swing1LimitDegrees=45
  swing1Motion=Locked
  swing2LimitDegrees=45
  swing2Motion=Locked
  twistLimitDegrees=45
  twistMotion=Locked
 }
 {bone1Name=dyn_Cloth_volume_r_C03, bone2Name=dyn_Cloth_volume_r_C02, swing1LimitDegrees=50, swing2LimitDegrees=15, twistLimitDegrees=15}
 {bone1Name=dyn_Cloth_volume_r_C02, bone2Name=dyn_Cloth_volume_r_C01, swing1LimitDegrees=30, swing2LimitDegrees=15, twistLimitDegrees=15}
 {bone1Name=dyn_Cloth_volume_r_C01, bone2Name=dyn_Cloth_volume_r_C, swing1LimitDegrees=15, swing2LimitDegrees=5, twistLimitDegrees=5}
 {
  bone1Name=dyn_Cloth_volume_r_C
  bone2Name=pelvis
  swing1LimitDegrees=45
  swing1Motion=Locked
  swing2LimitDegrees=45
  swing2Motion=Locked
  twistLimitDegrees=45
  twistMotion=Locked
 }
 {bone1Name=dyn_Cloth_volume_r_B03, bone2Name=dyn_Cloth_volume_r_B02, swing1LimitDegrees=30, swing2LimitDegrees=15, twistLimitDegrees=15}
 {bone1Name=dyn_Cloth_volume_r_B02, bone2Name=dyn_Cloth_volume_r_B01, swing1LimitDegrees=30, swing2LimitDegrees=15, twistLimitDegrees=15}
 {
  bone1Name=dyn_Cloth_volume_r_B01
  bone2Name=dyn_Cloth_volume_r_B
  swing1LimitDegrees=45
  swing1Motion=Locked
  swing2LimitDegrees=45
  swing2Motion=Locked
  twistLimitDegrees=45
  twistMotion=Locked
 }
 {
  bone1Name=dyn_Cloth_volume_r_B
  bone2Name=pelvis
  swing1LimitDegrees=45
  swing1Motion=Locked
  swing2LimitDegrees=45
  swing2Motion=Locked
  twistLimitDegrees=45
  twistMotion=Locked
 }
 {bone1Name=dyn_Belt_a02, bone2Name=dyn_Belt_a01, swing1LimitDegrees=15, swing2LimitDegrees=2, twistLimitDegrees=10}
 {bone1Name=dyn_Belt_a01, bone2Name=dyn_Belt_base, swing1LimitDegrees=1, swing2LimitDegrees=1, twistLimitDegrees=1}
]
