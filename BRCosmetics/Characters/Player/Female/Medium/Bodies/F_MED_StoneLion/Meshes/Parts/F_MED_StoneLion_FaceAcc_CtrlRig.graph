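# Control Rig for the female StoneLion face accessory that applies pose-driven corrective offsets to hair bones.
# Spherical pose readers on head and right upper arm write alphas, some multiplied into combined look-and-aim weights.
# These alphas weight Modify Transforms on hair master, front and "hant" bones, with crouch variables scaling interpolated corrections.
# Look-left, look-right and skydiving alphas are computed but drive nothing here.
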ControlRigBlueprint /BRCosmetics/Characters/Player/Female/Medium/Bodies/F_MED_StoneLion/Meshes/Parts/F_MED_StoneLion_FaceAcc_CtrlRig
graphs [
 ContainedGraph {nodes [
   n1 = RigVMFunctionEntryNode()
   SequenceExecution_12 = &RigVMFunction_Sequence(ExecuteContext=@n1.ExecuteContext)
   SequenceExecution_12_1 = &RigVMFunction_Sequence(ExecuteContext=@SequenceExecution_12.B)
   SequenceExecution_12_2 = &RigVMFunction_Sequence(ExecuteContext=@SequenceExecution_12_1.B)
   SequenceExecution_12_3 = &RigVMFunction_Sequence(ExecuteContext=@SequenceExecution_12_2.B)
   RigVMFunctionReturnNode(
    A=@SequenceExecution_12.A
    B=@SequenceExecution_12_1.A
    C=@SequenceExecution_12_2.A
    D=@SequenceExecution_12_3.A
    E=@SequenceExecution_12_3.B
   )
  ]}
 ContainedGraph {nodes [
   n1 = RigVMFunctionEntryNode()
   SequenceExecution_4 = &RigVMFunction_Sequence(ExecuteContext=@n1.ExecuteContext)
   SequenceExecution_4_1 = &RigVMFunction_Sequence(ExecuteContext=@SequenceExecution_4.B)
   SequenceExecution_4_2 = &RigVMFunction_Sequence(ExecuteContext=@SequenceExecution_4_1.B)
   RigVMFunctionReturnNode(A=@SequenceExecution_4.A, B=@SequenceExecution_4_1.A, C=@SequenceExecution_4_2.A, D=@SequenceExecution_4_2.B)
  ]}
 ContainedGraph {nodes [
   n1 = RigVMFunctionEntryNode()
   SequenceExecution_3 = &RigVMFunction_Sequence(ExecuteContext=@n1.ExecuteContext)
   SequenceExecution_3_1 = &RigVMFunction_Sequence(ExecuteContext=@SequenceExecution_3.B)
   SequenceExecution_3_2 = &RigVMFunction_Sequence(ExecuteContext=@SequenceExecution_3_1.B)
   RigVMFunctionReturnNode(A=@SequenceExecution_3.A, B=@SequenceExecution_3_1.A, C=@SequenceExecution_3_2.A, D=@SequenceExecution_3_2.B)
  ]}
 ContainedGraph {nodes [
   n1 = RigVMFunctionEntryNode()
   SequenceExecution_2 = &RigVMFunction_Sequence(ExecuteContext=@n1.ExecuteContext)
   SequenceExecution_2_1 = &RigVMFunction_Sequence(ExecuteContext=@SequenceExecution_2.B)
   SequenceExecution_2_2 = &RigVMFunction_Sequence(ExecuteContext=@SequenceExecution_2_1.B)
   RigVMFunctionReturnNode(A=@SequenceExecution_2.A, B=@SequenceExecution_2_1.A, C=@SequenceExecution_2_2.A, D=@SequenceExecution_2_2.B)
  ]}
 ContainedGraph {nodes [
   n1 = RigVMFunctionEntryNode()
   SequenceExecution_7 = &RigVMFunction_Sequence(ExecuteContext=@n1.ExecuteContext)
   SequenceExecution_7_1 = &RigVMFunction_Sequence(ExecuteContext=@SequenceExecution_7.B)
   SequenceExecution_7_2 = &RigVMFunction_Sequence(ExecuteContext=@SequenceExecution_7_1.B)
   RigVMFunctionReturnNode(A=@SequenceExecution_7.A, B=@SequenceExecution_7_1.A, C=@SequenceExecution_7_2.A, D=@SequenceExecution_7_2.B)
  ]}
 ContainedGraph {nodes [
   n1 = RigVMFunctionEntryNode()
   SequenceExecution = &RigVMFunction_Sequence(ExecuteContext=@n1.ExecuteContext)
   SequenceExecution_1 = &RigVMFunction_Sequence(ExecuteContext=@SequenceExecution.B)
   SequenceExecution_2 = &RigVMFunction_Sequence(ExecuteContext=@SequenceExecution_1.B)
   RigVMFunctionReturnNode(A=@SequenceExecution.A, B=@SequenceExecution_1.A, C=@SequenceExecution_2.A, D=@SequenceExecution_2.B)
  ]}
 ContainedGraph {nodes [
   n1 = RigVMFunctionEntryNode()
   SequenceExecution_8 = &RigVMFunction_Sequence(ExecuteContext=@n1.ExecuteContext)
   SequenceExecution_8_1 = &RigVMFunction_Sequence(ExecuteContext=@SequenceExecution_8.B)
   SequenceExecution_8_2 = &RigVMFunction_Sequence(ExecuteContext=@SequenceExecution_8_1.B)
   RigVMFunctionReturnNode(A=@SequenceExecution_8.A, B=@SequenceExecution_8_1.A, C=@SequenceExecution_8_2.A, D=@SequenceExecution_8_2.B)
  ]}
 ContainedGraph {nodes [
   n1 = RigVMFunctionEntryNode()
   SequenceExecution_6_1 = &RigVMFunction_Sequence(ExecuteContext=@n1.ExecuteContext)
   SequenceExecution_6_1_1 = &RigVMFunction_Sequence(ExecuteContext=@SequenceExecution_6_1.B)
   SequenceExecution_6_1_2 = &RigVMFunction_Sequence(ExecuteContext=@SequenceExecution_6_1_1.B)
   RigVMFunctionReturnNode(A=@SequenceExecution_6_1.A, B=@SequenceExecution_6_1_1.A, C=@SequenceExecution_6_1_2.A, D=@SequenceExecution_6_1_2.B)
  ]}
 RigVMFunctionLibrary {links [], nodes []}
 RigVMModel {nodes [
   RigUnit_BeginExecution = &RigUnit_BeginExecution()
   SequenceExecution_12 = RigVMAggregateNode(ExecuteContext=@RigUnit_BeginExecution.ExecutePin)
   SequenceExecution_3 = RigVMAggregateNode(ExecuteContext=@SequenceExecution_12.A)
   SequenceExecution_4 = RigVMAggregateNode(ExecuteContext=@SequenceExecution_3.A)
   SequenceExecution_7 = RigVMAggregateNode(ExecuteContext=@SequenceExecution_4.A)
   n1 = &RigUnit_SphericalPoseReader(
    ActiveRegionScaleFactors=(PositiveWidth=1,NegativeWidth=1,PositiveHeight=1,NegativeHeight=1)
    ActiveRegionSize=0.065
    Debug=(bDrawDebug=false,DebugScale=25,DebugSegments=20,DebugThickness=0.25)
    DriverAxis=(X=-1,Y=0,Z=0)
    DriverItem=(Type=Bone,Name="upperarm_r")
    FalloffRegionScaleFactors=(PositiveWidth=1,NegativeWidth=1,PositiveHeight=1,NegativeHeight=1)
    FalloffSize=0.125
    OptionalParentItem=(Type=Bone,Name="None")
    RotationOffset=(X=60,Y=0,Z=65)
    ExecutePin=@SequenceExecution_7.A
   )
   > RigVMVariableNode(Value=@n1.OutputParam) {variable=upperarm_r_SPR_back_alpha}
   n2 = &RigUnit_SphericalPoseReader(
    ActiveRegionScaleFactors=(PositiveWidth=1.2,NegativeWidth=1,PositiveHeight=1,NegativeHeight=1.5)
    ActiveRegionSize=0.075
    Debug=(bDrawDebug=false,DebugScale=25,DebugSegments=20,DebugThickness=0.25)
    DriverAxis=(X=-1,Y=0,Z=0)
    DriverItem=(Type=Bone,Name="upperarm_r")
    FalloffRegionScaleFactors=(PositiveWidth=1,NegativeWidth=1,PositiveHeight=1,NegativeHeight=1)
    FalloffSize=0.2
    OptionalParentItem=(Type=Bone)
    RotationOffset=(X=-60,Y=0,Z=-15)
    ExecutePin=@SequenceExecution_7.D
   )
   > RigVMVariableNode(Value=@n2.OutputParam) {variable=upperarm_r_rifle_up_alpha}
   n3 = &RigUnit_SphericalPoseReader(
    ActiveRegionScaleFactors=(PositiveWidth=1,NegativeWidth=1,PositiveHeight=1,NegativeHeight=1)
    ActiveRegionSize=0.06
    Debug=(bDrawDebug=false,DebugScale=25,DebugSegments=20,DebugThickness=0.25)
    DriverAxis=(X=-1,Y=0,Z=0)
    DriverItem=(Type=Bone,Name="upperarm_r")
    FalloffRegionScaleFactors=(PositiveWidth=1,NegativeWidth=1,PositiveHeight=1,NegativeHeight=1)
    FalloffSize=0.15
    OptionalParentItem=(Type=Bone,Name="spine_05")
    RotationOffset=(X=-90,Y=0,Z=40)
    ExecutePin=@SequenceExecution_7.B
   )
   > RigVMVariableNode(Value=@n3.OutputParam) {variable=uppearm_r_SPR_aim_up_two_hands_alpha}
   n4 = &RigUnit_SphericalPoseReader(
    ActiveRegionScaleFactors=(PositiveWidth=2,NegativeWidth=2,PositiveHeight=1,NegativeHeight=1)
    ActiveRegionSize=0.05
    Debug=(bDrawDebug=false,DebugScale=25,DebugSegments=20,DebugThickness=0.25)
    DriverAxis=(X=-1,Y=0,Z=0)
    DriverItem=(Type=Bone,Name="upperarm_r")
    FalloffRegionScaleFactors=(PositiveWidth=1,NegativeWidth=1,PositiveHeight=0.75,NegativeHeight=0.75)
    FalloffSize=0.2
    FlipWidthScaling=false
    OptionalParentItem=(Type=Bone)
    RotationOffset=(X=15,Y=150,Z=0)
    ExecutePin=@SequenceExecution_7.C
   )
   > RigVMVariableNode(Value=@n4.OutputParam) {variable=upperarm_r_skydiving}
   SequenceExecution_2 = RigVMAggregateNode(ExecuteContext=@SequenceExecution_4.B)
   n5 = &RigUnit_SphericalPoseReader(
    ActiveRegionScaleFactors=(PositiveWidth=1,NegativeWidth=0.5,PositiveHeight=1.5,NegativeHeight=1.5)
    ActiveRegionSize=0.1
    Debug=(bDrawDebug=false,DebugScale=25,DebugSegments=20,DebugThickness=0.25)
    DriverAxis=(X=0,Y=1,Z=0)
    DriverItem=(Type=Bone,Name="head")
    FalloffRegionScaleFactors=(PositiveWidth=1,NegativeWidth=0.7,PositiveHeight=1.5,NegativeHeight=1.5)
    FalloffSize=0.4
    FlipWidthScaling=false
    OptionalParentItem=(Type=Bone)
    RotationOffset=(X=90,Y=0,Z=-60)
    ExecutePin=@SequenceExecution_2.A
   )
   > RigVMVariableNode(Value=@n5.OutputParam) {variable=head_SPR_look_up_alpha}
   n6 = &RigUnit_SphericalPoseReader(
    ActiveRegionScaleFactors=(PositiveWidth=2,NegativeWidth=0.5,PositiveHeight=2,NegativeHeight=2)
    ActiveRegionSize=0.075
    Debug=(bDrawDebug=false,DebugScale=25,DebugSegments=20,DebugThickness=0.25)
    DriverAxis=(X=0,Y=1,Z=0)
    DriverItem=(Type=Bone,Name="head")
    FalloffRegionScaleFactors=(PositiveWidth=1,NegativeWidth=0.7,PositiveHeight=1,NegativeHeight=1)
    FalloffSize=0.2
    FlipHeightScaling=false
    FlipWidthScaling=true
    OptionalParentItem=(Type=Bone)
    RotationOffset=(X=90,Y=0,Z=12.5)
    ExecutePin=@SequenceExecution_2.B
   )
   > RigVMVariableNode(Value=@n6.OutputParam) {variable=head_SPR_look_down_alpha}
   n7 = &RigUnit_SphericalPoseReader(
    ActiveRegionScaleFactors=(PositiveWidth=0.75,NegativeWidth=0.75,PositiveHeight=1,NegativeHeight=0.5)
    ActiveRegionSize=0.1
    Debug=(bDrawDebug=false,DebugScale=25,DebugSegments=20,DebugThickness=0.25)
    DriverAxis=(X=0,Y=1,Z=0)
    DriverItem=(Type=Bone,Name="head")
    FalloffRegionScaleFactors=(PositiveWidth=1,NegativeWidth=1,PositiveHeight=1,NegativeHeight=0.7)
    FalloffSize=0.2
    OptionalParentItem=(Type=Bone)
    RotationOffset=(X=120,Y=0,Z=0)
    ExecutePin=@SequenceExecution_2.C
   )
   > RigVMVariableNode(Value=@n7.OutputParam) {variable=head_SPR_look_left_alpha}
   n8 = &RigUnit_SphericalPoseReader(
    ActiveRegionScaleFactors=(PositiveWidth=0.75,NegativeWidth=0.75,PositiveHeight=1,NegativeHeight=0.5)
    ActiveRegionSize=0.1
    Debug=(bDrawDebug=false,DebugScale=25,DebugSegments=20,DebugThickness=0.25)
    DriverAxis=(X=0,Y=1,Z=0)
    DriverItem=(Type=Bone,Name="head")
    FalloffRegionScaleFactors=(PositiveWidth=1,NegativeWidth=1,PositiveHeight=1,NegativeHeight=0.7)
    FalloffSize=0.2
    FlipHeightScaling=true
    FlipWidthScaling=false
    OptionalParentItem=(Type=Bone)
    RotationOffset=(X=60,Y=0,Z=0)
    ExecutePin=@SequenceExecution_2.D
   )
   > RigVMVariableNode(Value=@n8.OutputParam) {variable=head_SPR_look_right_alpha}
   RigVMVariableNode(
    Value=&RigVMFunction_MathFloatMul(B=RigVMVariableNode() {variable=uppearm_r_SPR_aim_up_two_hands_alpha}.Value, A=RigVMVariableNode() {variable=head_SPR_look_up_alpha}.Value) {template="Multiply::Execute(in A,in B,out Result)"}.Result
    ExecuteContext=@SequenceExecution_3.B
   ) {variable=look_up_and_left_two_hands_multiplied_alpha}
   RigVMVariableNode(
    Value=&RigVMFunction_MathFloatMul(A=RigVMVariableNode() {variable=head_SPR_look_up_alpha}.Value, B=RigVMVariableNode() {variable=upperarm_r_rifle_up_alpha}.Value) {template="Multiply::Execute(in A,in B,out Result)"}.Result
    ExecuteContext=@SequenceExecution_3.C
   ) {variable=look_up_two_hands_rifle_multiplied_alpha}
   RigVMVariableNode(
    Value=&RigVMFunction_MathFloatMul(B=RigVMVariableNode() {variable=head_SPR_look_down_alpha}.Value, A=RigVMVariableNode() {variable=upperarm_r_SPR_back_alpha}.Value) {template="Multiply::Execute(in A,in B,out Result)"}.Result
    ExecuteContext=@SequenceExecution_3.D
   ) {variable=look_down_upperarm_r_back_multiplied_alpha}
   SequenceExecution = RigVMAggregateNode(ExecuteContext=@SequenceExecution_12.B)
   n9 = RigVMVariableNode() {variable=head_SPR_look_up_alpha}
   &RigUnit_ModifyTransforms(
    ItemToModify=((Item=(Type=Bone,Name="dyn_hair_r_Front_00"),Transform=(Rotation=(X=0,Y=0,Z=0,W=1),Translation=(X=0,Y=0,Z=0),Scale3D=(X=1,Y=1,Z=1))))
    ItemToModify.0.Transform.Rotation=&RigVMFunction_MathQuaternionFromEuler(Euler=(X=0,Y=0,Z=-5)).Result
    ExecutePin=@SequenceExecution.B
    Weight=@n9.Value
   )
   &RigUnit_ModifyTransforms(
    ItemToModify=((Item=(Type=Bone,Name="dyn_hair_l_Front_MSTR"),Transform=(Rotation=(X=0,Y=0,Z=0,W=1),Translation=(X=0,Y=0,Z=0),Scale3D=(X=1,Y=1,Z=1))))
    ExecutePin=@SequenceExecution.C
    Weight=@n9.Value
   )
   &RigUnit_ModifyTransforms(
    ItemToModify="((Item=(Type=Bone,Name=\"hant\"),Transform=(Rotation=(X=0,Y=0,Z=0.0436,W=0.999),Translation=(X=2,Y=0,Z=0),Scale3D=(X=1,Y=1,Z=1))),(Item=(Type=Bone,Name=\"dyn_hair_side_r_MSTR\"),Transform=(Rotation=(X=0,Y=0.4226,Z=0,W=0.9063),Translation=(X=0,Y=5,Z=-5),Scale3D ...(truncated)"
    ExecutePin=@SequenceExecution.D
    Weight=RigVMVariableNode() {variable=head_SPR_look_up_alpha}.Value
   )
   &RigUnit_ModifyTransforms(
    ItemToModify="((Item=(Type=Bone,Name=\"dyn_hair_bk_MSTR\"),Transform=(Rotation=(X=0,Y=0,Z=0,W=1),Translation=(X=2,Y=-4,Z=0),Scale3D=(X=1,Y=1,Z=1))),(Item=(Type=Bone,Name=\"dyn_hair_l_MSTR\"),Transform=(Rotation=(X=0,Y=0,Z=0,W=1),Translation=(X=2,Y=-4,Z=0),Scale3D=(X=1,Y=1,Z ...(truncated)"
    ItemToModify.0.Transform.Rotation=&RigVMFunction_MathQuaternionFromEuler(Euler=(X=0,Y=0,Z=-5), RotationOrder=XYZ).Result
    ItemToModify.2.Transform.Rotation=&RigVMFunction_MathQuaternionFromEuler(Euler=(X=0,Y=0,Z=-5), RotationOrder=XYZ).Result
    ItemToModify.1.Transform.Rotation=&RigVMFunction_MathQuaternionFromEuler(Euler=(X=0,Y=0,Z=-5), RotationOrder=XYZ).Result
    Weight=RigVMVariableNode() {variable=head_SPR_look_up_alpha}.Value
    ExecutePin=@SequenceExecution.A
   )
   > &RigUnit_ModifyTransforms(
    ItemToModify="((Item=(Type=Bone,Name=\"dyn_hair_side_l_MSTR\"),Transform=(Rotation=(X=0,Y=0,Z=0,W=1),Translation=(X=0,Y=-3,Z=1),Scale3D=(X=1,Y=1,Z=1))),(Item=(Type=Bone,Name=\"dyn_hair_side_r_MSTR\"),Transform=(Rotation=(X=0,Y=0,Z=0,W=1),Translation=(X=-1,Y=-10,Z=-2),Scale3 ...(truncated)"
    ItemToModify.0.Transform.Rotation=&RigVMFunction_MathQuaternionFromEuler(Euler=(X=0,Y=0,Z=-5), RotationOrder=XYZ).Result
    ItemToModify.1.Transform.Rotation=&RigVMFunction_MathQuaternionFromEuler(Euler=(X=0,Y=0,Z=-5), RotationOrder=XYZ).Result
    Weight=RigVMVariableNode() {variable=head_SPR_look_up_alpha}.Value
   )
   n10 = /Script/RigVM.RigVMFunction_MathFloatMax(
    B=RigVMVariableNode() {variable=look_up_two_hands_rifle_multiplied_alpha}.Value
    A=RigVMVariableNode() {variable=look_up_and_left_two_hands_multiplied_alpha}.Value
   ) {is=struct RigVMUnitNode, template="Maximum::Execute(in A,in B,out Result)"}
   > &RigUnit_ModifyTransforms(
    ItemToModify=((Item=(Type=Bone,Name="dyn_hair_bk_MSTR"),Transform=(Rotation=(X=0,Y=0,Z=0,W=1),Translation=(X=0,Y=0,Z=-5),Scale3D=(X=1,Y=1,Z=1))))
    ItemToModify.0.Transform.Rotation=&RigVMFunction_MathQuaternionFromEuler(Euler=(X=0,Y=0,Z=0), RotationOrder=XYZ).Result
    Weight=@n10.Result
   )
   > &RigUnit_ModifyTransforms(
    ItemToModify="((Item=(Type=Bone,Name=\"dyn_hair_side_r_MSTR\"),Transform=(Rotation=(X=0,Y=0,Z=0,W=1),Translation=(X=-3,Y=-3,Z=0),Scale3D=(X=1,Y=1,Z=1))),(Item=(Type=Bone,Name=\"dyn_hair_r_MSTR\"),Transform=(Rotation=(X=0,Y=0,Z=0,W=1),Translation=(X=-1,Y=1,Z=-5),Scale3D=(X=1 ...(truncated)"
    ItemToModify.0.Transform.Rotation=&RigVMFunction_MathQuaternionFromEuler(Euler=(X=5,Y=0,Z=-10), RotationOrder=XYZ).Result
    ItemToModify.1.Transform.Rotation=&RigVMFunction_MathQuaternionFromEuler(Euler=(X=10,Y=0,Z=-10), RotationOrder=XYZ).Result
    Weight=@n10.Result
   )
   > &RigUnit_ModifyTransforms(
    ItemToModify="((Item=(Type=Bone,Name=\"dyn_hair_side_l_MSTR\"),Transform=(Rotation=(X=0,Y=0,Z=0,W=1),Translation=(X=1,Y=0,Z=1),Scale3D=(X=1,Y=1,Z=1))),(Item=(Type=Bone,Name=\"dyn_hair_l_MSTR\"),Transform=(Rotation=(X=0,Y=0,Z=0,W=1),Translation=(X=0,Y=0,Z=-2),Scale3D=(X=1,Y= ...(truncated)"
    ItemToModify.0.Transform.Rotation=&RigVMFunction_MathQuaternionFromEuler(Euler=(X=0,Y=10,Z=10), RotationOrder=XYZ).Result
    Weight=@n10.Result
   )
   > &RigUnit_ModifyTransforms(
    ItemToModify=((Item=(Type=Bone,Name="dyn_hair_l_Front_MSTR"),Transform=(Rotation=(X=0,Y=0,Z=0,W=1),Translation=(X=2,Y=-2,Z=0),Scale3D=(X=1,Y=1,Z=1))))
    ItemToModify.0.Transform.Rotation=&RigVMFunction_MathQuaternionFromEuler(Euler=(X=90,Y=0,Z=15), RotationOrder=XYZ).Result
    Weight=@n10.Result
   )
   SequenceExecution_8 = RigVMAggregateNode(ExecuteContext=@SequenceExecution_12.C)
   n11 = RigVMVariableNode() {variable=look_up_and_left_two_hands_multiplied_alpha}
   &RigUnit_ModifyTransforms(
    ItemToModify=((Item=(Type=Bone,Name="dyn_hair_r_MSTR"),Transform=(Rotation=(X=0,Y=0,Z=0,W=1),Translation=(X=1,Y=2,Z=-1),Scale3D=(X=1,Y=1,Z=1))))
    ItemToModify.0.Transform.Rotation=&RigVMFunction_MathQuaternionFromEuler(Euler=(X=0,Y=5,Z=-5), RotationOrder=XYZ).Result
    ExecutePin=@SequenceExecution_8.A
    Weight=@n11.Value
   )
   > &RigUnit_ModifyTransforms(
    ItemToModify=((Item=(Type=Bone,Name="dyn_hair_side_r_MSTR"),Transform=(Rotation=(X=0,Y=0,Z=0,W=1),Translation=(X=0,Y=5,Z=0),Scale3D=(X=1,Y=1,Z=1))))
    ItemToModify.0.Transform.Rotation=&RigVMFunction_MathQuaternionFromEuler(Euler=(X=0,Y=0,Z=0), RotationOrder=XYZ).Result
    Weight=@n11.Value
   )
   n12 = &RigUnit_SphericalPoseReader(
    ActiveRegionScaleFactors=(PositiveWidth=2,NegativeWidth=0.5,PositiveHeight=2,NegativeHeight=2)
    ActiveRegionSize=0.075
    Debug=(bDrawDebug=false,DebugScale=25,DebugSegments=20,DebugThickness=0.25)
    DriverAxis=(X=1,Y=0,Z=0)
    DriverItem=(Type=Bone,Name="head")
    FalloffRegionScaleFactors=(PositiveWidth=1,NegativeWidth=0.4,PositiveHeight=1,NegativeHeight=1)
    FalloffSize=0.25
    FlipHeightScaling=false
    FlipWidthScaling=true
    OptionalParentItem=(Type=Bone,Name="spine_05")
    RotationOffset=(X=-70,Y=5,Z=0)
    ExecutePin=@SequenceExecution_8.D
   )
   n13 = &RigVMFunction_AlphaInterp(
    InterpSpeedDecreasing=10
    InterpSpeedIncreasing=10
    bClampResult=true
    Value=@n12.OutputParam
    Scale=RigVMVariableNode() {variable=crouch}.Value
   )
   > &RigUnit_ModifyTransforms(
    ItemToModify="((Item=(Type=Bone,Name=\"dyn_hair_bk_MSTR\"),Transform=(Rotation=(X=0,Y=0,Z=0,W=1),Translation=(X=3,Y=-3,Z=-2),Scale3D=(X=1,Y=1,Z=1))),(Item=(Type=Bone,Name=\"dyn_hair_l_MSTR\"),Transform=(Rotation=(X=0,Y=0,Z=0,W=1),Translation=(X=1,Y=-5,Z=-2),Scale3D=(X=1,Y=1 ...(truncated)"
    Weight=@n13.Result
    ItemToModify.0.Transform.Rotation=&RigVMFunction_MathQuaternionFromEuler(Euler=(X=0,Y=0,Z=0), RotationOrder=XYZ).Result
    ItemToModify.2.Transform.Rotation=&RigVMFunction_MathQuaternionFromEuler(Euler=(X=0,Y=0,Z=0), RotationOrder=XYZ).Result
    ItemToModify.1.Transform.Rotation=&RigVMFunction_MathQuaternionFromEuler(Euler=(X=0,Y=0,Z=0), RotationOrder=XYZ).Result
   )
   > &RigUnit_ModifyTransforms(
    ItemToModify="((Item=(Type=Bone,Name=\"dyn_hair_side_l_MSTR\"),Transform=(Rotation=(X=0,Y=0,Z=0,W=1),Translation=(X=-3,Y=-5,Z=-2),Scale3D=(X=1,Y=1,Z=1))),(Item=(Type=Bone,Name=\"dyn_hair_side_r_MSTR\"),Transform=(Rotation=(X=0,Y=0,Z=0,W=1),Translation=(X=-3,Y=-5,Z=-2),Scale ...(truncated)"
    Weight=@n13.Result
    ItemToModify.1.Transform.Rotation=&RigVMFunction_MathQuaternionFromEuler(Euler=(X=0,Y=0,Z=-5), RotationOrder=XYZ).Result
    ItemToModify.0.Transform.Rotation=&RigVMFunction_MathQuaternionFromEuler(Euler=(X=0,Y=0,Z=-5), RotationOrder=XYZ).Result
   )
   &RigUnit_ModifyTransforms(
    ItemToModify=((Item=(Type=Bone,Name="dyn_hair_bk_MSTR"),Transform=(Rotation=(X=0,Y=0,Z=0,W=1),Translation=(X=0,Y=2,Z=-2),Scale3D=(X=1,Y=1,Z=1))))
    ItemToModify.0.Transform.Rotation=&RigVMFunction_MathQuaternionFromEuler(Euler=(X=0,Y=5,Z=0), RotationOrder=XYZ).Result
    Weight=RigVMVariableNode() {variable=look_up_and_left_two_hands_multiplied_alpha}.Value
    ExecutePin=@SequenceExecution_8.B
   )
   n14 = RigVMVariableNode() {variable=look_up_and_left_two_hands_multiplied_alpha}
   &RigUnit_ModifyTransforms(
    ItemToModify=((Item=(Type=Bone,Name="dyn_hair_l_MSTR"),Transform=(Rotation=(X=0,Y=0,Z=0,W=1),Translation=(X=-0.5,Y=-2,Z=-4),Scale3D=(X=1,Y=1,Z=1))))
    ItemToModify.0.Transform.Rotation=&RigVMFunction_MathQuaternionFromEuler(Euler=(X=0,Y=0,Z=-5), RotationOrder=XYZ).Result
    ExecutePin=@SequenceExecution_8.C
    Weight=@n14.Value
   )
   > &RigUnit_ModifyTransforms(
    ItemToModify=((Item=(Type=Bone,Name="dyn_hair_side_l_MSTR"),Transform=(Rotation=(X=0,Y=0,Z=0,W=1),Translation=(X=-1,Y=1,Z=1),Scale3D=(X=1,Y=1,Z=1))))
    ItemToModify.0.Transform.Rotation=&RigVMFunction_MathQuaternionFromEuler(Euler=(X=0,Y=-10,Z=-10), RotationOrder=XYZ).Result
    Weight=@n14.Value
   )
   SequenceExecution_6_1 = RigVMAggregateNode(ExecuteContext=@SequenceExecution_12.D)
   n15 = RigVMVariableNode() {variable=head_SPR_look_down_alpha}
   &RigUnit_ModifyTransforms(
    ItemToModify=((Item=(Type=Bone,Name="dyn_hair_r_MSTR"),Transform=(Rotation=(X=0,Y=0,Z=0,W=1),Translation=(X=0,Y=-1,Z=0),Scale3D=(X=1,Y=1,Z=1))))
    ItemToModify.0.Transform.Rotation=&RigVMFunction_MathQuaternionFromEuler(Euler=(X=5,Y=5,Z=0)).Result
    ExecutePin=@SequenceExecution_6_1.B
    Weight=@n15.Value
   )
   > &RigUnit_ModifyTransforms(
    ItemToModify=((Item=(Type=Bone,Name="dyn_hair_side_r_MSTR"),Transform=(Rotation=(X=0,Y=0,Z=0,W=1),Translation=(X=-1,Y=-3,Z=-2),Scale3D=(X=1,Y=1,Z=1))))
    ItemToModify.0.Transform.Rotation=&RigVMFunction_MathQuaternionFromEuler(Euler=(X=5,Y=5,Z=0)).Result
    Weight=@n15.Value
   )
   &RigUnit_ModifyTransforms(
    ItemToModify=((Item=(Type=Bone,Name="dyn_hair_bk_MSTR"),Transform=(Rotation=(X=0,Y=0,Z=0,W=1),Translation=(X=0,Y=0,Z=0),Scale3D=(X=1,Y=1,Z=1))))
    ItemToModify.0.Transform.Rotation=&RigVMFunction_MathQuaternionFromEuler(Euler=(X=0,Y=0,Z=0)).Result
    ExecutePin=@SequenceExecution_6_1.C
    Weight=RigVMVariableNode() {variable=head_SPR_look_down_alpha}.Value
   )
   &RigUnit_ModifyTransforms(
    ItemToModify=((Item=(Type=Bone,Name="dyn_hair_l_MSTR"),Transform=(Rotation=(X=0,Y=0,Z=0,W=1),Translation=(X=0,Y=0,Z=0),Scale3D=(X=1,Y=1,Z=1))))
    ItemToModify.0.Transform.Rotation=&RigVMFunction_MathQuaternionFromEuler(Euler=(X=0,Y=0,Z=0)).Result
    Weight=RigVMVariableNode() {variable=look_down_upperarm_r_back_multiplied_alpha}.Value
    ExecutePin=@SequenceExecution_6_1.D
   )
   AlphaInterp = &RigVMFunction_AlphaInterp(
    InterpSpeedDecreasing=10
    InterpSpeedIncreasing=10
    bClampResult=true
    Value=&RigVMFunction_MathFloatMax(
     B=RigVMVariableNode() {variable=look_up_two_hands_rifle_multiplied_alpha}.Value
     A=RigVMVariableNode() {variable=look_up_and_left_two_hands_multiplied_alpha}.Value
    ) {template="Maximum::Execute(in A,in B,out Result)"}.Result
    Scale=RigVMVariableNode() {variable=crouch move}.Value
   )
   &RigUnit_ModifyTransforms(
    ItemToModify="((Item=(Type=Bone,Name=\"dyn_hair_bk_MSTR\"),Transform=(Rotation=(X=0,Y=0,Z=0,W=1),Translation=(X=4,Y=-3,Z=0),Scale3D=(X=1,Y=1,Z=1))),(Item=(Type=Bone,Name=\"dyn_hair_l_MSTR\"),Transform=(Rotation=(X=0,Y=0,Z=0,W=1),Translation=(X=3,Y=-4,Z=0),Scale3D=(X=1,Y=1,Z ...(truncated)"
    Weight=@AlphaInterp.Result
    ItemToModify.0.Transform.Rotation=&RigVMFunction_MathQuaternionFromEuler(Euler=(X=0,Y=0,Z=0), RotationOrder=XYZ).Result
    ItemToModify.2.Transform.Rotation=&RigVMFunction_MathQuaternionFromEuler(Euler=(X=0,Y=-5,Z=0), RotationOrder=XYZ).Result
    ItemToModify.1.Transform.Rotation=&RigVMFunction_MathQuaternionFromEuler(Euler=(X=0,Y=5,Z=0), RotationOrder=XYZ).Result
    ExecutePin=@SequenceExecution_12.E
   )
   > &RigUnit_ModifyTransforms(
    ItemToModify="((Item=(Type=Bone,Name=\"dyn_hair_side_l_MSTR\"),Transform=(Rotation=(X=0,Y=0,Z=0,W=1),Translation=(X=0,Y=-4,Z=0),Scale3D=(X=1,Y=1,Z=1))),(Item=(Type=Bone,Name=\"dyn_hair_side_r_MSTR\"),Transform=(Rotation=(X=0,Y=0,Z=0,W=1),Translation=(X=0,Y=-4,Z=0),Scale3D=( ...(truncated)"
    Weight=@AlphaInterp.Result
    ItemToModify.1.Transform.Rotation=&RigVMFunction_MathQuaternionFromEuler(Euler=(X=0,Y=-5,Z=-5), RotationOrder=XYZ).Result
    ItemToModify.0.Transform.Rotation=&RigVMFunction_MathQuaternionFromEuler(Euler=(X=0,Y=5,Z=-5), RotationOrder=XYZ).Result
   )
   RigVMVariableNode() {variable=look_up_and_left_two_hands_multiplied_alpha}
   RigVMVariableNode() {variable=look_down_upperarm_r_back_multiplied_alpha}
   RigVMVariableNode() {variable=look_down_upperarm_r_back_multiplied_alpha}
  ]}
]
hierarchy [
 root {loc=<0 0 0>, rot=<0 0 0 1>}
 pelvis {parent=root, loc=<0 -0.5301 98.0571>, rot=<0.0127 0.707 -0.0127 -0.707>}
 spine_01 {parent=pelvis, loc=<6.1208 -0.9328 0>, rot=<0 0 0.034 -0.9994>}
 spine_02 {parent=spine_01, loc=<6.5996 -0.2095 0>, rot=<0 0 0.0238 -0.9997>}
 spine_03 {parent=spine_02, loc=<6.4914 -0.1237 0>, rot=<0 0 0.0149 -0.9999>}
 spine_04 {parent=spine_03, loc=<6.9598 0.0794 0>, rot=<0 0 -0.0132 -0.9999>}
 spine_05 {parent=spine_04, loc=<6.9987 0 0>, rot=<0 0 -0.0015 -1>}
 neck_01 {parent=spine_05, loc=<11.3138 0.1529 0>, rot=<0 0 -0.2003 -0.9797>}
 neck_02 {parent=neck_01, loc=<4.5627 0 0>, rot=<0 0 0.0504 -0.9987>}
 head {parent=neck_02, loc=<5.3914 0 0>, rot=<0 0 0.115 -0.9934>}
 dyn_hair_side_l_MSTR {parent=head, loc=<3.1076 -2.3134 -6.1821>, rot=<0.0039 -0.9527 -0.0013 -0.304>}
 dyn_hair_side_l_01 {parent=dyn_hair_side_l_MSTR, loc=<3.7508 0 -1.1856>, rot=<0.0093 -0.0593 0.1552 -0.9861>}
 dyn_hair_side_l_02 {parent=dyn_hair_side_l_01, loc=<4.4233 0 0>, rot=<0.0571 -0.07 0.3623 -0.9277>}
 dyn_hair_side_l_03 {parent=dyn_hair_side_l_02, loc=<6.9485 0 0>, rot=<0.1064 -0.1201 -0.3217 -0.9331>}
 dyn_hair_side_l_04 {parent=dyn_hair_side_l_03, loc=<6.1091 0 0>, rot=<0.0044 -0.0141 -0.091 -0.9957>}
 dyn_hair_side_r_MSTR {parent=head, loc=<3.0998 -1.3199 6.182>, rot=<0.004 -0.9632 0.0011 0.2689>}
 dyn_hair_side_r_01 {parent=dyn_hair_side_r_MSTR, loc=<4.5941 0.1845 0.2417>, rot=<-0.0387 0.1288 0.285 -0.949>}
 dyn_hair_side_r_02 {parent=dyn_hair_side_r_01, loc=<4.6046 0 0>, rot=<0.0904 -0.0814 0.2498 -0.9606>}
 dyn_hair_side_r_03 {parent=dyn_hair_side_r_02, loc=<5.5047 0 0>, rot=<-0.1734 0.1639 -0.2673 -0.9336>}
 dyn_hair_side_r_04 {parent=dyn_hair_side_r_03, loc=<6.1472 0 0>, rot=<-0.0123 0.0325 -0.1711 -0.9846>}
 dyn_hair_r_Front_MSTR {parent=head, loc=<-0.0836 0.9058 6.7342>, rot=<-0.0558 -0.9952 -0.0052 -0.0805>}
 dyn_hair_r_Front_00 {parent=dyn_hair_r_Front_MSTR, loc=<0.4477 -0.1171 -0.1893>, rot=<-0.0054 -0.1675 0.0869 -0.982>}
 dyn_hair_r_Front_01 {parent=dyn_hair_r_Front_00, loc=<5.4453 0.81 -0.3225>, rot=<-0.0006 -0.0191 -0.0856 -0.9961>}
 dyn_hair_r_Front_02 {parent=dyn_hair_r_Front_01, loc=<4.1116 1.9505 0.2984>, rot=<0.0306 0.1045 -0.1681 -0.9797>}
 dyn_hair_r_Front_03 {parent=dyn_hair_r_Front_02, loc=<5.9973 0.8507 0.4482>, rot=<-0.0001 -0.0003 0.0533 -0.9986>}
 dyn_hair_r_MSTR {parent=head, loc=<4.2877 -3.7149 6.0243>, rot=<0.0663 -0.9912 -0.0067 0.1146>}
 dyn_hair_r_00 {parent=dyn_hair_r_MSTR, loc=<4.4758 -2.3083 0.3358>, rot=<-0.0102 0.0396 0.1262 -0.9912>}
 dyn_hair_r_01 {parent=dyn_hair_r_00, loc=<7.194 0 0>, rot=<-0.0233 0.0506 0.0526 -0.9971>}
 dyn_hair_r_02 {parent=dyn_hair_r_01, loc=<6.1472 0 0>, rot=<-0.0106 0.0328 -0.171 -0.9847>}
 dyn_hair_bk_MSTR {parent=head, loc=<4.4961 -4.542 0.0291>, rot=<0.1718 -0.9851 0 0>}
 dyn_hair_bk_00 {parent=dyn_hair_bk_MSTR, loc=<4.2135 -0.7738 0.0291>, rot=<0 0 0.069 -0.9976>}
 dyn_hair_bk_01 {parent=dyn_hair_bk_00, loc=<7.7147 0 0>, rot=<0 0 -0.0149 -0.9999>}
 dyn_hair_bk_02 {parent=dyn_hair_bk_01, loc=<6.6501 0 0>, rot=<0 0 -0.1461 -0.9893>}
 dyn_hair_l_MSTR {parent=head, loc=<4.2741 -4.3095 -6.1145>, rot=<0.0919 -0.9926 0.0067 -0.0796>}
 dyn_hair_l_00 {parent=dyn_hair_l_MSTR, loc=<4.4664 -1.5204 -0.1644>, rot=<0.0181 -0.0525 0.1543 -0.9865>}
 dyn_hair_l_01 {parent=dyn_hair_l_00, loc=<8.1281 0 0>, rot=<0.0116 -0.0246 -0.0469 -0.9985>}
 dyn_hair_l_02 {parent=dyn_hair_l_01, loc=<6.1091 0 0>, rot=<0.0044 -0.0141 -0.091 -0.9957>}
 dyn_hair_front_r_MSTR {parent=head, loc=<3.3172 1.5541 -6.3683>, rot=<-0.0029 0.7071 0.7071 -0.0029>}
 dyn_hair_front_r_01 {parent=dyn_hair_front_r_MSTR, loc=<0 0 0>, rot=<0 0 0 1>}
 dyn_hair_front_r_02 {parent=dyn_hair_front_r_01, loc=<2.9654 0 0>, rot=<0 0 0 1>}
 dyn_strands_MSTR {parent=head, loc=<7.6561 9.7758 3.1166>, rot=<0.2732 -0.354 0.8936 -0.0386>}
 dyn_strands_01 {parent=dyn_strands_MSTR, loc=<0 0 0>, rot=<0 0 0 1>}
 dyn_strands_02 {parent=dyn_strands_01, loc=<4.9747 0 0>, rot=<-0.0158 0.1129 -0.3255 -0.9387>}
 dyn_strands_03 {parent=dyn_strands_02, loc=<4.6322 -0.2368 -0.7656>, rot=<-0.0697 0.106 -0.0411 -0.9911>}
 dyn_strands_04 {parent=dyn_strands_03, loc=<3.5487 0 0>, rot=<0.0836 0.4026 -0.0339 -0.9109>}
 dyn_strands_05 {parent=dyn_strands_04, loc=<6.0555 0 0>, rot=<0.2055 0.0968 -0.1604 -0.9606>}
 dyn_strands_l_01 {parent=dyn_strands_MSTR, loc=<0 0 0>, rot=<0 0 0 1>}
 dyn_hair_l_Front_MSTR {parent=head, loc=<2.0052 -0.1639 -6.8572>, rot=<-0.0558 -0.9952 -0.0052 -0.0805>}
 dyn_hair_l_Front_00 {parent=dyn_hair_l_Front_MSTR, loc=<0.4477 -0.1171 -0.1893>, rot=<-0.0886 0.0365 0.0878 -0.9915>}
 dyn_hair_l_Front_01 {parent=dyn_hair_l_Front_00, loc=<5.4453 0.81 -0.3225>, rot=<0.0042 -0.0741 -0.0855 -0.9936>}
 dyn_hair_l_Front_02 {parent=dyn_hair_l_Front_01, loc=<4.1116 1.9505 0.2984>, rot=<0.0555 -0.0439 -0.1615 -0.9843>}
 dyn_hair_l_Front_03 {parent=dyn_hair_l_Front_02, loc=<5.9973 0.8507 0.4482>, rot=<-0.0001 -0.0003 0.0533 -0.9986>}
 hant {parent=head, loc=<8.7308 0.0722 0>, rot=<0 0 0.0604 -0.9982>}
 hairStrand_r_MSTR {parent=head, loc=<3.2851 6.4805 -6.4527>, rot=<0.0414 -0.9917 -0.1168 -0.0331>}
 hairStrand_r_01 {parent=hairStrand_r_MSTR, loc=<0 0 0>, rot=<0 0 0 1>}
 dyn_strands_MSTRA {parent=head, loc=<9.3483 9.4429 2.4581>, rot=<0.2732 -0.354 0.8936 -0.0386>}
 dyn_strands_l_A01 {parent=dyn_strands_MSTRA, loc=<0 0 0>, rot=<0 0 0 1>}
 dyn_strands_l_A02 {parent=dyn_strands_l_A01, loc=<5.0078 0.0589 -0.3944>, rot=<0 0 0 1>}
 hairStrand_r_MSTRA {parent=head, loc=<9.7455 8.8629 -4.4897>, rot=<0.0414 -0.9917 -0.1168 -0.0331>}
 hairStrand_r_A01 {parent=hairStrand_r_MSTRA, loc=<0 0 0>, rot=<0 0 0 1>}
 clavicle_l {parent=spine_05, loc=<5.4831 -0.1053 -3.2946>, rot=<0.0465 -0.7289 -0.009 -0.6829>}
 upperarm_l {parent=clavicle_l, loc=<9.8658 -1.2136 -0.4105>, rot=<0.0164 -0.3245 -0.0269 -0.9454>}
 clavicle_r {parent=spine_05, loc=<5.4831 -0.1053 3.2946>, rot=<0.7289 0.0465 -0.6829 0.009>}
 upperarm_r {parent=clavicle_r, loc=<-9.8658 1.2136 0.4105>, rot=<0.0164 -0.3245 -0.0269 -0.9454>}
]
variables [
 clavicle_r_SPR_back_alpha {type=double, default=0}
 crouch {type=double, default=0}
 crouch move {type=double, default=0}
 head_SPR_look_down_alpha {type=double, default=0}
 head_SPR_look_left_alpha {type=double, default=0}
 head_SPR_look_right_alpha {type=double, default=0}
 head_SPR_look_up_alpha {type=double, default=0}
 look_down_upperarm_r_back_multiplied_alpha {type=double, default=0}
 look_up_and_left_two_hands_multiplied_alpha {type=double, default=0}
 look_up_two_hands_rifle_multiplied_alpha {type=double, default=0}
 uppearm_r_SPR_aim_up_two_hands_alpha {type=double, default=0}
 upperarm_r_SPR_back_alpha {type=double, default=0}
 upperarm_r_rifle_up_alpha {type=double, default=0}
 upperarm_r_skydiving {type=double, default=0}
]
