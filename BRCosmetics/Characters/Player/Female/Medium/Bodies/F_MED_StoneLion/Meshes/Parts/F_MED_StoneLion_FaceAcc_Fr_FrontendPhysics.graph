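# Physics asset for the frontend variant of the female medium StoneLion face accessory, driving its dangling front hair chains.
# Capsules on head, neck, upper spine, clavicles and upper arms form body collision volumes.
# Left and right front hair chains hang from head via free MSTR joints, with limited swing and tight twist below; left allows wider swing.
# The left side-hair, clavicle and upper-arm bodies have no constraints.
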PhysicsAsset /BRCosmetics/Characters/Player/Female/Medium/Bodies/F_MED_StoneLion/Meshes/Parts/F_MED_StoneLion_FaceAcc_Fr_FrontendPhysics
bodies {
 clavicle_l [{shapeType=Capsule, length=5.6669, radius=6.0644, center=<5.681 -0.1497 -1.7263>, rotation=<rot 85.606 0 0>}]
 clavicle_r [{shapeType=Capsule, length=4.1669, radius=6.5644, center=<-6.5041 -0.1497 2.1395>, rotation=<rot 85.606 0 0>}]
 dyn_hair_l_Front_00 [{shapeType=Capsule, length=1.875, radius=3.125, shapeName=dyn_hair_r_Front_01_capsule, center=<3 0 0>, rotation=<rot 90 0 0>}]
 dyn_hair_l_Front_01 [{
   shapeType=Capsule
   length=2.5625
   radius=5.4375
   shapeName=dyn_hair_r_Front_02_capsule
   center=<3.761 0.731 1.7972>
   rotation=<rot 90 -121.3016 -168.2922>
  }]
 dyn_hair_l_Front_02 [{
   shapeType=Capsule
   length=1.3125
   radius=3.4375
   shapeName=dyn_hair_r_Front_03_capsule
   center=<3.6912 2.8655 0.2579>
   rotation=<rot 90 -179.5524 175.4474>
  }]
 dyn_hair_l_Front_03 [{shapeType=Capsule, length=5.375, radius=4.375, shapeName=dyn_hair_r_Front_04_capsule, center=<4.0109 0.5114 0>, rotation=<rot 90 0 0>}]
 dyn_hair_l_Front_MSTR [{shapeType=Capsule, length=0.505, radius=0.505, shapeName=dyn_hair_r_Front_MSTR_capsule, rotation=<rot 0 -90 0>}]
 dyn_hair_r_Front_00 [{shapeType=Capsule, length=1.875, radius=3.125, shapeName=dyn_hair_r_Front_01_capsule, center=<3 0 0>, rotation=<rot 90 0 0>}]
 dyn_hair_r_Front_01 [{
   shapeType=Capsule
   length=5
   radius=3.4375
   shapeName=dyn_hair_r_Front_02_capsule
   center=<2.0077 2.2774 0.6594>
   rotation=<rot 90 180 154.6872>
  }]
 dyn_hair_r_Front_02 [{
   shapeType=Capsule
   length=1.3125
   radius=3.4375
   shapeName=dyn_hair_r_Front_03_capsule
   center=<3.7642 1.014 0>
   rotation=<rot 90 -172.875 172.1249>
  }]
 dyn_hair_r_Front_03 [{shapeType=Capsule, length=4.375, radius=2.375, shapeName=dyn_hair_r_Front_04_capsule, center=<3.5 1 0>, rotation=<rot 90 0 0>}]
 dyn_hair_r_Front_MSTR [{shapeType=Capsule, length=0.505, radius=0.505, shapeName=dyn_hair_r_Front_MSTR_capsule, rotation=<rot 0 -90 0>}]
 dyn_hair_side_l_02 [{
   shapeType=Capsule
   length=4.5
   radius=4.5
   shapeName=dyn_hair_side_l_02_capsule
   center=<2.497 -1.9739 -2.0953>
   rotation=<rot 50 0.0001 -40>
  }]
 head [{shapeType=Capsule, length=4.5796, radius=5.2593, shapeName=head_capsule, center=<-1.3548 1.1749 -0.5075>, rotation=<rot 90 0 5.0001>}]
 neck_01 [{shapeType=Capsule, length=1.5, radius=5.5675, shapeName=neck_01_capsule, center=<0 1.9981 0>, rotation=<rot 0 -90.0005 0>}]
 spine_04 [{shapeType=Capsule, length=12.1566, radius=11.5126, shapeName=spine_04_capsule, center=<3.8752 -0.5041 0>, rotation=<rot 0 -92.6847 0>}]
 spine_05 [{shapeType=Capsule, length=14.7816, radius=11.0126, shapeName=spine_04_capsule, center=<0.0525 -1.893 0>, rotation=<rot 0 -92.6847 0>}]
 upperarm_l [{shapeType=Capsule, length=13.4229, radius=4.1429, shapeName=upperarm_r_capsule, center=<5.2485 1.0639 -0.2913>, rotation=<rot 0 -90 90>}]
 upperarm_r [{shapeType=Capsule, length=20.4229, radius=4.1429, shapeName=upperarm_r_capsule, center=<-11.013 1.25 -1.2498>, rotation=<rot 0 -90 90>}]
}
constraints [
 {
  bone1Name=head
  bone2Name=neck_01
  swing1LimitDegrees=45
  swing1Motion=Free
  swing2LimitDegrees=45
  swing2Motion=Free
  twistLimitDegrees=45
  twistMotion=Free
 }
 {
  bone1Name=spine_05
  bone2Name=spine_04
  swing1LimitDegrees=45
  swing1Motion=Free
  swing2LimitDegrees=45
  swing2Motion=Free
  twistLimitDegrees=45
  twistMotion=Free
 }
 {
  bone1Name=clavicle_r
  bone2Name=spine_05
  swing1LimitDegrees=45
  swing1Motion=Free
  swing2LimitDegrees=45
  swing2Motion=Free
  twistLimitDegrees=45
  twistMotion=Free
 }
 {
  bone1Name=upperarm_r
  bone2Name=clavicle_r
  swing1LimitDegrees=45
  swing1Motion=Free
  swing2LimitDegrees=45
  swing2Motion=Free
  twistLimitDegrees=45
  twistMotion=Free
 }
 {
  bone1Name=neck_01
  bone2Name=spine_05
  swing1LimitDegrees=45
  swing1Motion=Free
  swing2LimitDegrees=45
  swing2Motion=Free
  twistLimitDegrees=45
  twistMotion=Free
 }
 {bone1Name=dyn_hair_r_Front_03, bone2Name=dyn_hair_r_Front_02, swing1LimitDegrees=35, swing2LimitDegrees=15, twistLimitDegrees=5}
 {
  bone1Name=dyn_hair_r_Front_MSTR
  bone2Name=head
  swing1LimitDegrees=45
  swing1Motion=Free
  swing2LimitDegrees=45
  swing2Motion=Free
  twistLimitDegrees=45
  twistMotion=Free
 }
 {bone1Name=dyn_hair_r_Front_00, bone2Name=dyn_hair_r_Front_MSTR, swing1LimitDegrees=45, swing2LimitDegrees=45, twistLimitDegrees=5}
 {bone1Name=dyn_hair_r_Front_01, bone2Name=dyn_hair_r_Front_00, swing1LimitDegrees=35, swing2LimitDegrees=15, twistLimitDegrees=5}
 {bone1Name=dyn_hair_r_Front_02, bone2Name=dyn_hair_r_Front_01, swing1LimitDegrees=35, swing2LimitDegrees=15, twistLimitDegrees=5}
 {
  bone1Name=dyn_hair_l_Front_MSTR
  bone2Name=head
  swing1LimitDegrees=45
  swing1Motion=Free
  swing2LimitDegrees=45
  swing2Motion=Free
  twistLimitDegrees=45
  twistMotion=Free
 }
 {bone1Name=dyn_hair_l_Front_00, bone2Name=dyn_hair_l_Front_MSTR, swing1LimitDegrees=45, swing2LimitDegrees=45, twistLimitDegrees=5}
 {bone1Name=dyn_hair_l_Front_01, bone2Name=dyn_hair_l_Front_00, swing1LimitDegrees=45, swing2LimitDegrees=15, twistLimitDegrees=5}
 {bone1Name=dyn_hair_l_Front_02, bone2Name=dyn_hair_l_Front_01, swing1LimitDegrees=60, swing2LimitDegrees=15, twistLimitDegrees=5}
 {bone1Name=dyn_hair_l_Front_03, bone2Name=dyn_hair_l_Front_02, swing1LimitDegrees=60, swing2LimitDegrees=15, twistLimitDegrees=5}
]
