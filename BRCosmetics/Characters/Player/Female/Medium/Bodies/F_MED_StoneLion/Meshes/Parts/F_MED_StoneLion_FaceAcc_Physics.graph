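# Physics asset for the StoneLion female medium FaceAcc cosmetic mesh, defining capsule bodies and constraints for its dynamic hair bones.
# Seven back, side and front hair chains hang from MSTR bones jointed to the head, with tighter angular limits along each chain.
# Head, neck, upper spine, clavicle and upper-arm capsules form the body volume.
# Cross-links tie neighboring chains together, including an angularly locked right side-to-back link.
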PhysicsAsset /BRCosmetics/Characters/Player/Female/Medium/Bodies/F_MED_StoneLion/Meshes/Parts/F_MED_StoneLion_FaceAcc_Physics
bodies {
 clavicle_l [{shapeType=Capsule, length=1.9169, radius=8.8144, center=<6.4709 1.1493 -2.1395>, rotation=<rot 85.606 0 0>}]
 clavicle_r [{shapeType=Capsule, length=4.6669, radius=8.0644, center=<-7.6181 -0.4649 2.3477>, rotation=<rot 85.606 0 0>}]
 dyn_hair_bk_00 [{shapeType=Capsule, length=3.3125, radius=4.4375, shapeName=dyn_Hair_Back_03_capsule, center=<5 -2 0>, rotation=<rot 90 0 0>}]
 dyn_hair_bk_01 [{shapeType=Capsule, length=2.5, radius=5.25, shapeName=dyn_Hair_Back_04_capsule, center=<5 -2 0>, rotation=<rot 90 0 0>}]
 dyn_hair_bk_02 [{shapeType=Capsule, length=4.1875, radius=3.5625, shapeName=dyn_Hair_Back_05_capsule, center=<5 -2 0>, rotation=<rot 90 0 0>}]
 dyn_hair_bk_MSTR [{shapeType=Capsule, length=0.505, radius=0.505, shapeName=dyn_Hair_Back_MSTR_capsule}]
 dyn_hair_l_00 [{shapeType=Capsule, length=0.3125, radius=5.3125, shapeName=dyn_Hair_BackL_03_capsule, center=<5 -2 0>, rotation=<rot 90 0 0>}]
 dyn_hair_l_01 [{shapeType=Capsule, length=0.375, radius=5.25, shapeName=dyn_Hair_BackL_04_capsule, center=<5 -2 0>, rotation=<rot 90 0 0>}]
 dyn_hair_l_02 [{shapeType=Capsule, length=2, radius=3.625, shapeName=dyn_Hair_BackL_05_capsule, center=<5 -2 0>, rotation=<rot 90 0 0>}]
 dyn_hair_l_Front_00 [{shapeType=Capsule, length=1.875, radius=3.125, shapeName=dyn_hair_r_Front_01_capsule, center=<3 0 0>, rotation=<rot 90 0 0>}]
 dyn_hair_l_Front_01 [{
   shapeType=Capsule
   length=2.0625
   radius=5.9375
   shapeName=dyn_hair_r_Front_02_capsule
   center=<4.2625 0.0148 0.4682>
   rotation=<rot 90 165.9638 130.9634>
  }]
 dyn_hair_l_Front_02 [{
   shapeType=Capsule
   length=0
   radius=4.9375
   shapeName=dyn_hair_r_Front_03_capsule
   center=<3.816 1.4395 0>
   rotation=<rot 90 -179.5524 175.4474>
  }]
 dyn_hair_l_Front_03 [{shapeType=Capsule, length=4.875, radius=1.875, shapeName=dyn_hair_r_Front_04_capsule, center=<3.5 0.3397 0>, rotation=<rot 90 0 0>}]
 dyn_hair_l_Front_MSTR [{shapeType=Capsule, length=0.505, radius=0.505, shapeName=dyn_hair_r_Front_MSTR_capsule, rotation=<rot 0 -90 0>}]
 dyn_hair_l_MSTR [{shapeType=Capsule, length=0.505, radius=0.505, shapeName=dyn_Hair_BackL_MSTR_capsule}]
 dyn_hair_r_00 [{shapeType=Capsule, length=1.3125, radius=4.3125, shapeName=dyn_Hair_BackR_03_capsule, center=<5 -2 0>, rotation=<rot 90 0 0>}]
 dyn_hair_r_01 [{shapeType=Capsule, length=0.9375, radius=4.6875, shapeName=dyn_Hair_BackR_04_capsule, center=<5 -2 0>, rotation=<rot 90 0 0>}]
 dyn_hair_r_02 [{shapeType=Capsule, length=1.5, radius=4.625, shapeName=dyn_Hair_BackR_05_capsule, center=<5 -2 0>, rotation=<rot 90 0 0>}]
 dyn_hair_r_Front_00 [{shapeType=Capsule, length=1.875, radius=3.125, shapeName=dyn_hair_r_Front_01_capsule, center=<3 0 0>, rotation=<rot 90 0 0>}]
 dyn_hair_r_Front_01 [{
   shapeType=Capsule
   length=5
   radius=3.4375
   shapeName=dyn_hair_r_Front_02_capsule
   center=<2.0077 2.2774 0.6594>
   rotation=<rot 90 180 154.6872>
  }]
 dyn_hair_r_Front_02 [{
   shapeType=Capsule
   length=1.3125
   radius=3.4375
   shapeName=dyn_hair_r_Front_03_capsule
   center=<3.7642 1.014 0>
   rotation=<rot 90 -172.875 172.1249>
  }]
 dyn_hair_r_Front_03 [{shapeType=Capsule, length=4.375, radius=2.375, shapeName=dyn_hair_r_Front_04_capsule, center=<3.5 1 0>, rotation=<rot 90 0 0>}]
 dyn_hair_r_Front_MSTR [{shapeType=Capsule, length=0.505, radius=0.505, shapeName=dyn_hair_r_Front_MSTR_capsule, rotation=<rot 0 -90 0>}]
 dyn_hair_r_MSTR [{shapeType=Capsule, length=0.505, radius=0.505, shapeName=dyn_Hair_BackR_MSTR_capsule}]
 dyn_hair_side_l_01 [{shapeType=Capsule, length=3.0625, radius=4, shapeName=dyn_Hair_BackL_03_capsule, center=<4 0 2>, rotation=<rot 90 0 0>}]
 dyn_hair_side_l_02 [{shapeType=Capsule, length=5, radius=3, shapeName=dyn_Hair_BackL_03_capsule, center=<4 1 0>, rotation=<rot 90 0 0>}]
 dyn_hair_side_l_03 [{shapeType=Capsule, length=2.6875, radius=4.0625, shapeName=dyn_Hair_BackL_04_capsule, center=<4 -2 0>, rotation=<rot 90 0 0>}]
 dyn_hair_side_l_04 [{shapeType=Capsule, length=2.6875, radius=2.9375, shapeName=dyn_Hair_BackL_05_capsule, center=<5 -2 0>, rotation=<rot 90 0 0>}]
 dyn_hair_side_l_MSTR [{shapeType=Capsule, length=0.505, radius=0.505, shapeName=dyn_Hair_BackL_MSTR_capsule}]
 dyn_hair_side_r_01 [{
   shapeType=Capsule
   length=1.5
   radius=5.9375
   shapeName=dyn_Hair_BackL_03_capsule
   center=<2.685 0.4519 -1.2173>
   rotation=<rot 90 180 -157.5001>
  }]
 dyn_hair_side_r_02 [{shapeType=Capsule, length=2, radius=6, shapeName=dyn_Hair_BackL_03_capsule, center=<5.511 -0.7441 -0.9742>, rotation=<rot 90 0 0>}]
 dyn_hair_side_r_03 [{shapeType=Capsule, length=0.875, radius=5.875, shapeName=dyn_Hair_BackL_04_capsule, center=<4 -1.7253 0>, rotation=<rot 90 0 0>}]
 dyn_hair_side_r_04 [{shapeType=Capsule, length=1.625, radius=4, shapeName=dyn_Hair_BackL_05_capsule, center=<5 -2.0042 0>, rotation=<rot 90 0 0>}]
 dyn_hair_side_r_05 [{
   shapeType=Capsule
   length=1.5
   radius=6.5
   shapeName=dyn_Hair_BackL_05_capsule
   center=<-1.0956 0.37 -0.9239>
   rotation=<rot -6.2717 -22.5516 -50.4676>
  }]
 dyn_hair_side_r_MSTR [{shapeType=Capsule, length=0.505, radius=0.505, shapeName=dyn_Hair_BackR_MSTR_capsule}]
 head [{shapeType=Capsule, length=3.0796, radius=6.7593, shapeName=head_capsule, center=<2.3025 0.855 -0.5075>, rotation=<rot 90 0 5.0001>}]
 neck_01 [{shapeType=Capsule, length=4.255, radius=5.5675, shapeName=neck_01_capsule, rotation=<rot 0 -90 0>}]
 spine_04 [{shapeType=Capsule, length=4.5941, radius=10.0751, shapeName=spine_04_capsule, center=<2.5892 0.4958 0>, rotation=<rot 0 -92.6847 0>}]
 spine_05 [{shapeType=Capsule, length=15.7816, radius=10.0126, shapeName=spine_04_capsule, center=<0.0821 0.0358 0>, rotation=<rot 0 -92.6847 0>}]
 upperarm_l [{shapeType=Capsule, length=13.4229, radius=4.1429, shapeName=upperarm_r_capsule, center=<5.2485 1.0639 -0.2913>, rotation=<rot 0 -90 90>}]
 upperarm_r [{shapeType=Capsule, length=20.4229, radius=4.1429, shapeName=upperarm_r_capsule, center=<-11.013 1.25 -1.2498>, rotation=<rot 0 -90 90>}]
}
constraints [
 {
  bone1Name=head
  bone2Name=neck_01
  swing1LimitDegrees=45
  swing1Motion=Free
  swing2LimitDegrees=45
  swing2Motion=Free
  twistLimitDegrees=45
  twistMotion=Free
 }
 {
  bone1Name=spine_05
  bone2Name=spine_04
  swing1LimitDegrees=45
  swing1Motion=Free
  swing2LimitDegrees=45
  swing2Motion=Free
  twistLimitDegrees=45
  twistMotion=Free
 }
 {
  bone1Name=clavicle_l
  bone2Name=spine_05
  swing1LimitDegrees=45
  swing1Motion=Free
  swing2LimitDegrees=45
  swing2Motion=Free
  twistLimitDegrees=45
  twistMotion=Free
 }
 {
  bone1Name=upperarm_l
  bone2Name=clavicle_l
  swing1LimitDegrees=45
  swing1Motion=Free
  swing2LimitDegrees=45
  swing2Motion=Free
  twistLimitDegrees=45
  twistMotion=Free
 }
 {
  bone1Name=clavicle_r
  bone2Name=spine_05
  swing1LimitDegrees=45
  swing1Motion=Free
  swing2LimitDegrees=45
  swing2Motion=Free
  twistLimitDegrees=45
  twistMotion=Free
 }
 {
  bone1Name=upperarm_r
  bone2Name=clavicle_r
  swing1LimitDegrees=45
  swing1Motion=Free
  swing2LimitDegrees=45
  swing2Motion=Free
  twistLimitDegrees=45
  twistMotion=Free
 }
 {
  bone1Name=dyn_hair_bk_MSTR
  bone2Name=head
  swing1LimitDegrees=45
  swing1Motion=Free
  swing2LimitDegrees=45
  swing2Motion=Free
  twistLimitDegrees=45
  twistMotion=Free
 }
 {bone1Name=dyn_hair_bk_00, bone2Name=dyn_hair_bk_MSTR, swing1LimitDegrees=50, swing2LimitDegrees=15, twistLimitDegrees=10}
 {bone1Name=dyn_hair_bk_01, bone2Name=dyn_hair_bk_00, swing1LimitDegrees=50, swing2LimitDegrees=15, twistLimitDegrees=10}
 {bone1Name=dyn_hair_bk_02, bone2Name=dyn_hair_bk_01, swing1LimitDegrees=50, swing2LimitDegrees=15, twistLimitDegrees=10}
 {
  bone1Name=dyn_hair_l_MSTR
  bone2Name=head
  swing1LimitDegrees=45
  swing1Motion=Free
  swing2LimitDegrees=45
  swing2Motion=Free
  twistLimitDegrees=45
  twistMotion=Free
 }
 {bone1Name=dyn_hair_l_00, bone2Name=dyn_hair_l_MSTR, swing1LimitDegrees=50, swing2LimitDegrees=15, twistLimitDegrees=10}
 {bone1Name=dyn_hair_l_01, bone2Name=dyn_hair_l_00, swing1LimitDegrees=50, swing2LimitDegrees=15, twistLimitDegrees=10}
 {bone1Name=dyn_hair_l_02, bone2Name=dyn_hair_l_01, swing1LimitDegrees=50, swing2LimitDegrees=15, twistLimitDegrees=10}
 {
  bone1Name=dyn_hair_r_MSTR
  bone2Name=head
  swing1LimitDegrees=45
  swing1Motion=Free
  swing2LimitDegrees=45
  swing2Motion=Free
  twistLimitDegrees=45
  twistMotion=Free
 }
 {bone1Name=dyn_hair_r_00, bone2Name=dyn_hair_r_MSTR, swing1LimitDegrees=50, swing2LimitDegrees=10, twistLimitDegrees=10}
 {bone1Name=dyn_hair_r_01, bone2Name=dyn_hair_r_00, swing1LimitDegrees=50, swing2LimitDegrees=10, twistLimitDegrees=10}
 {bone1Name=dyn_hair_r_02, bone2Name=dyn_hair_r_01, swing1LimitDegrees=50, swing2LimitDegrees=10, twistLimitDegrees=10}
 {
  bone1Name=dyn_hair_side_l_MSTR
  bone2Name=head
  swing1LimitDegrees=45
  swing1Motion=Free
  swing2LimitDegrees=45
  swing2Motion=Free
  twistLimitDegrees=45
  twistMotion=Free
 }
 {bone1Name=dyn_hair_side_l_01, bone2Name=dyn_hair_side_l_MSTR, swing1LimitDegrees=50, swing2LimitDegrees=15, twistLimitDegrees=10}
 {bone1Name=dyn_hair_side_l_02, bone2Name=dyn_hair_side_l_01, swing1LimitDegrees=50, swing2LimitDegrees=15, twistLimitDegrees=10}
 {bone1Name=dyn_hair_side_l_03, bone2Name=dyn_hair_side_l_02, swing1LimitDegrees=50, swing2LimitDegrees=15, twistLimitDegrees=10}
 {bone1Name=dyn_hair_side_l_04, bone2Name=dyn_hair_side_l_03, swing1LimitDegrees=50, swing2LimitDegrees=15, twistLimitDegrees=10}
 {
  bone1Name=dyn_hair_side_r_MSTR
  bone2Name=head
  swing1LimitDegrees=45
  swing1Motion=Free
  swing2LimitDegrees=45
  swing2Motion=Free
  twistLimitDegrees=45
  twistMotion=Free
 }
 {bone1Name=dyn_hair_side_r_01, bone2Name=dyn_hair_side_r_MSTR, swing1LimitDegrees=50, swing2LimitDegrees=10, twistLimitDegrees=10}
 {bone1Name=dyn_hair_side_r_02, bone2Name=dyn_hair_side_r_01, swing1LimitDegrees=50, swing2LimitDegrees=10, twistLimitDegrees=10}
 {bone1Name=dyn_hair_side_r_03, bone2Name=dyn_hair_side_r_02, swing1LimitDegrees=50, swing2LimitDegrees=10, twistLimitDegrees=10}
 {bone1Name=dyn_hair_side_r_04, bone2Name=dyn_hair_side_r_03, swing1LimitDegrees=50, swing2LimitDegrees=10, twistLimitDegrees=10}
 {
  bone1Name=neck_01
  bone2Name=spine_05
  swing1LimitDegrees=45
  swing1Motion=Free
  swing2LimitDegrees=45
  swing2Motion=Free
  twistLimitDegrees=45
  twistMotion=Free
 }
 {bone1Name=dyn_hair_side_l_03, bone2Name=dyn_hair_l_01, swing1LimitDegrees=45, swing2LimitDegrees=45, twistLimitDegrees=45}
 {
  bone1Name=dyn_hair_side_r_03
  bone2Name=dyn_hair_r_01
  swing1LimitDegrees=15
  swing1Motion=Locked
  swing2LimitDegrees=15
  swing2Motion=Locked
  twistLimitDegrees=15
  twistMotion=Locked
 }
 {bone1Name=dyn_hair_r_01, bone2Name=dyn_hair_bk_01, swing1LimitDegrees=45, swing2LimitDegrees=45, twistLimitDegrees=45}
 {
  bone1Name=dyn_hair_l_Front_MSTR
  bone2Name=head
  swing1LimitDegrees=45
  swing1Motion=Free
  swing2LimitDegrees=45
  swing2Motion=Free
  twistLimitDegrees=45
  twistMotion=Free
 }
 {bone1Name=dyn_hair_l_Front_00, bone2Name=dyn_hair_l_Front_MSTR, swing1LimitDegrees=45, swing2LimitDegrees=45, twistLimitDegrees=5}
 {bone1Name=dyn_hair_l_Front_01, bone2Name=dyn_hair_l_Front_00, swing1LimitDegrees=45, swing2LimitDegrees=15, twistLimitDegrees=5}
 {bone1Name=dyn_hair_l_Front_02, bone2Name=dyn_hair_l_Front_01, swing1LimitDegrees=45, swing2LimitDegrees=15, twistLimitDegrees=5}
 {bone1Name=dyn_hair_l_Front_03, bone2Name=dyn_hair_l_Front_02, swing1LimitDegrees=45, swing2LimitDegrees=15, twistLimitDegrees=5}
 {
  bone1Name=dyn_hair_r_Front_MSTR
  bone2Name=head
  swing1LimitDegrees=45
  swing1Motion=Free
  swing2LimitDegrees=45
  swing2Motion=Free
  twistLimitDegrees=45
  twistMotion=Free
 }
 {bone1Name=dyn_hair_r_Front_00, bone2Name=dyn_hair_r_Front_MSTR, swing1LimitDegrees=45, swing2LimitDegrees=45, twistLimitDegrees=5}
 {bone1Name=dyn_hair_r_Front_01, bone2Name=dyn_hair_r_Front_00, swing1LimitDegrees=45, swing2LimitDegrees=15, twistLimitDegrees=5}
 {bone1Name=dyn_hair_r_Front_02, bone2Name=dyn_hair_r_Front_01, swing1LimitDegrees=45, swing2LimitDegrees=15, twistLimitDegrees=5}
 {bone1Name=dyn_hair_r_Front_03, bone2Name=dyn_hair_r_Front_02, swing1LimitDegrees=45, swing2LimitDegrees=15, twistLimitDegrees=5}
 {bone1Name=dyn_hair_side_r_05, bone2Name=dyn_hair_side_r_03, swing1LimitDegrees=45, swing2LimitDegrees=45, twistLimitDegrees=45}
]
